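# Sound cue for the BarnTowel air stop, classed as a gadget sound and using the in-flight air third-person attenuation asset.
# A mixer plays three layers at once with separate volumes: a stop whoosh passed through a modulator, a random fire whoosh and a random sputter burst.
# Both random nodes weight their variants equally, so repeated triggers vary, with the sputter layer mixed loudest.
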SoundCue /BarnTowelGameplay/Assets/Sounds/_various/BarnTowel_Airstop_Cue
root=Mixer_0
nodes [
 {id=Mixer_0, children [Modulator_0,Random_0,Random_1], properties {inputVolume [1,0.5,1.75]}}
 {id=Modulator_0, children [WavePlayer_0]}
 {id=Random_0, children [WavePlayer_1,WavePlayer_2,WavePlayer_3,WavePlayer_4], properties {weights [1,1,1,1]}}
 {id=Random_1, children [WavePlayer_5, WavePlayer_7, WavePlayer_8, WavePlayer_9, WavePlayer_10], properties {weights [1,1,1,1,1]}}
 {id=WavePlayer_0, properties {soundWaveAssetPtr=&/BarnTowel_Stop_Whoosh}}
 {id=WavePlayer_1, properties {soundWaveAssetPtr=&/BarnTowel_Fire_Whoosh_01}}
 {id=WavePlayer_10, properties {soundWaveAssetPtr=&/BarnTowel_Sputter_Burst_06}}
 {id=WavePlayer_2, properties {soundWaveAssetPtr=&/BarnTowel_Fire_Whoosh_02}}
 {id=WavePlayer_3, properties {soundWaveAssetPtr=&/BarnTowel_Fire_Whoosh_03}}
 {id=WavePlayer_4, properties {soundWaveAssetPtr=&/BarnTowel_Fire_Whoosh_04}}
 {id=WavePlayer_5, properties {soundWaveAssetPtr=&/BarnTowel_Sputter_Burst_01}}
 {id=WavePlayer_7, properties {soundWaveAssetPtr=&/BarnTowel_Sputter_Burst_03}}
 {id=WavePlayer_8, properties {soundWaveAssetPtr=&/BarnTowel_Sputter_Burst_04}}
 {id=WavePlayer_9, properties {soundWaveAssetPtr=&/BarnTowel_Sputter_Burst_05}}
]
properties {
 attenuationSettings=/BarnTowelGameplay/Assets/Sounds/Settings/BarnTowel_InFlightAir_3P_Att
 maxDistance=2097152
 soundClassObject=&Fort_Gadgets
}
