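# Sound cue for the BarnTowel bonk, built from two force-crash impact waves in the Fort_Gadgets sound class.
# A Random node picks one of the two waves with equal weight on each play.
# It plays at reduced volume with the BarnTowel ground attenuation asset.
# A concurrency override caps it at a few voices per owner, stopping the oldest, with a retrigger time.
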SoundCue /BarnTowelGameplay/Assets/Sounds/_various/BarnTowel_Bonk_Cue
root=Random_1
nodes [
 {id=Random_1, children [WavePlayer_0,WavePlayer_1], properties {weights [1,1]}}
 {id=WavePlayer_0, properties {soundWaveAssetPtr=&/BarnTowel_ForceCrash_Impt_01}}
 {id=WavePlayer_1, properties {soundWaveAssetPtr=&/BarnTowel_ForceCrash_Impt_02}}
]
properties {
 attenuationSettings=/BarnTowelGameplay/Assets/Sounds/Settings/BarnTowel_Ground_Att
 bOverrideConcurrency=true
 maxDistance=3900
 soundClassObject=&Fort_Gadgets
 volumeMultiplier=0.35
 concurrencyOverrides {bLimitToOwner=true, maxCount=2, resolutionRule=StopOldest, retriggerTime=1}
}
