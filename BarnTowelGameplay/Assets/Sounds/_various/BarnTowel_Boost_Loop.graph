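# Looping SoundCue for the BarnTowel gadget's in-air boost, with 3P attenuation and concurrency capped at one instance per owner.
# It mixes a randomly chosen wind loop with a randomly chosen exhaust loop, three variants each.
# Branches on islocal and isenemy route that shared mix: full volume for the local player, half for non-enemy others, and boosted for enemies.
# The whole cue plays with a raised pitch multiplier.
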SoundCue /BarnTowelGameplay/Assets/Sounds/_various/BarnTowel_Boost_Loop
root=Branch_2
nodes [
 {id=Branch_2, children [Modulator_3,Branch_3,Modulator_3], properties {boolParameterName=islocal}}
 {id=Branch_3, children [Mixer_1,Modulator_4,Mixer_1], properties {boolParameterName=isenemy}}
 {id=Mixer_0, children [Modulator_0,Modulator_1], properties {inputVolume [1,1]}}
 {id=Mixer_1, children [Modulator_5], properties {inputVolume [1.3]}}
 {id=Modulator_0, children [Random_0]}
 {id=Modulator_1, children [Random_1]}
 {id=Modulator_3, children [Mixer_0], properties {pitchMax=1, pitchMin=1, volumeMax=1, volumeMin=1}}
 {id=Modulator_4, children [Mixer_0], properties {pitchMax=1, pitchMin=1, volumeMax=0.5, volumeMin=0.5}}
 {id=Modulator_5, children [Mixer_0], properties {pitchMax=1, pitchMin=1, volumeMax=1, volumeMin=1}}
 {id=Random_0, children [WavePlayer_6,WavePlayer_7,WavePlayer_8], properties {weights [1,1,1]}}
 {id=Random_1, children [WavePlayer_9, WavePlayer_10, WavePlayer_11], properties {weights [1,1,1]}}
 {id=WavePlayer_10, properties {bLooping=true, soundWaveAssetPtr=&/BarnTowel_InAirLoop_Exhaust_02}}
 {id=WavePlayer_11, properties {bLooping=true, soundWaveAssetPtr=&/BarnTowel_InAirLoop_Exhaust_01}}
 {id=WavePlayer_6, properties {bLooping=true, soundWaveAssetPtr=&/BarnTowel_InAirLoop_Wind_03}}
 {id=WavePlayer_7, properties {bLooping=true, soundWaveAssetPtr=&/BarnTowel_InAirLoop_Wind_02}}
 {id=WavePlayer_8, properties {bLooping=true, soundWaveAssetPtr=&/BarnTowel_InAirLoop_Wind_01}}
 {id=WavePlayer_9, properties {bLooping=true, soundWaveAssetPtr=&/BarnTowel_InAirLoop_Exhaust_03}}
]
properties {
 attenuationSettings=/BarnTowelGameplay/Assets/Sounds/Settings/BarnTowel_InFlightAir_3P_Att
 bOverrideConcurrency=true
 maxDistance=2097152
 pitchMultiplier=1.2
 soundClassObject=&Fort_Gadgets
 concurrencyOverrides {bLimitToOwner=true, maxCount=1}
}
