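# Sound Cue for the BarnTowel gadget's descend sound, using in-flight air third-person attenuation and the Fort_Gadgets sound class.
# A mixer layers a modulated side whoosh with randomly picked fire whoosh, sputter burst and forward whoosh variations, weighting the side whoosh loudest.
# Branches on islocal and isenemy route that shared mix through volume modulators, halving volume for non-local, non-enemy sources and keeping full volume otherwise.
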SoundCue /BarnTowelGameplay/Assets/Sounds/_various/BarnTowel_Descend_Cue
root=Branch_2
nodes [
 {id=Branch_2, children [Modulator_2,Branch_3,Modulator_2], properties {boolParameterName=islocal}}
 {id=Branch_3, children [Modulator_4,Modulator_3,Modulator_4], properties {boolParameterName=isenemy}}
 {id=Mixer_0, children [Modulator_0,Random_0,Random_1,Random_2], properties {inputVolume [1,0.35,0.5,0.5]}}
 {id=Modulator_0, children [WavePlayer_0]}
 {id=Modulator_2, children [Mixer_0], properties {pitchMax=1, pitchMin=1, volumeMax=1, volumeMin=1}}
 {id=Modulator_3, children [Mixer_0], properties {pitchMax=1, pitchMin=1, volumeMax=0.5, volumeMin=0.5}}
 {id=Modulator_4, children [Mixer_0], properties {pitchMax=1, pitchMin=1, volumeMax=1, volumeMin=1}}
 {id=Random_0, children [WavePlayer_1,WavePlayer_2,WavePlayer_3,WavePlayer_4], properties {weights [1,1,1,1]}}
 {id=Random_1, children [WavePlayer_5, WavePlayer_7, WavePlayer_8, WavePlayer_9, WavePlayer_10], properties {weights [1,1,1,1,1]}}
 {id=Random_2, children [WavePlayer_11, WavePlayer_12, WavePlayer_13], properties {weights [1,1,1]}}
 {id=WavePlayer_0, properties {soundWaveAssetPtr=&/BarnTowel_Side_Whoosh}}
 {id=WavePlayer_1, properties {soundWaveAssetPtr=&/BarnTowel_Fire_Whoosh_01}}
 {id=WavePlayer_10, properties {soundWaveAssetPtr=&/BarnTowel_Sputter_Burst_06}}
 {id=WavePlayer_11, properties {soundWaveAssetPtr=&/BarnTowel_Forward_Whoosh_01}}
 {id=WavePlayer_12, properties {soundWaveAssetPtr=&/BarnTowel_Forward_Whoosh_02}}
 {id=WavePlayer_13, properties {soundWaveAssetPtr=&/BarnTowel_Forward_Whoosh_03}}
 {id=WavePlayer_2, properties {soundWaveAssetPtr=&/BarnTowel_Fire_Whoosh_02}}
 {id=WavePlayer_3, properties {soundWaveAssetPtr=&/BarnTowel_Fire_Whoosh_03}}
 {id=WavePlayer_4, properties {soundWaveAssetPtr=&/BarnTowel_Fire_Whoosh_04}}
 {id=WavePlayer_5, properties {soundWaveAssetPtr=&/BarnTowel_Sputter_Burst_01}}
 {id=WavePlayer_7, properties {soundWaveAssetPtr=&/BarnTowel_Sputter_Burst_03}}
 {id=WavePlayer_8, properties {soundWaveAssetPtr=&/BarnTowel_Sputter_Burst_04}}
 {id=WavePlayer_9, properties {soundWaveAssetPtr=&/BarnTowel_Sputter_Burst_05}}
]
properties {
 attenuationSettings=/BarnTowelGameplay/Assets/Sounds/Settings/BarnTowel_InFlightAir_3P_Att
 maxDistance=2097152
 soundClassObject=&Fort_Gadgets
 volumeMultiplier=1
}
