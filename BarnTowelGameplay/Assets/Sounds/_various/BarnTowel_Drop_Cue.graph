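# BarnTowel_Drop_Cue is a SoundCue for the sound of the BarnTowel gadget being dropped and landing.
# A mixer layers one of three equally weighted random metal landing waves with a drop wave, which a modulator pitches down within a narrow random range.
# It plays at reduced volume through the Fort_Gadgets sound class with weapon-drop distance attenuation.
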SoundCue /BarnTowelGameplay/Assets/Sounds/_various/BarnTowel_Drop_Cue
root=Mixer_0
nodes [
 {id=Mixer_0, children [Random_0,Modulator_0], properties {inputVolume [1,1]}}
 {id=Modulator_0, children [WavePlayer_4], properties {pitchMax=0.9, pitchMin=0.8}}
 {id=Random_0, children [WavePlayer_1,WavePlayer_2,WavePlayer_3], properties {weights [1,1,1]}}
 {id=WavePlayer_1, properties {soundWaveAssetPtr=&/BarnTowel_Metal_Land_03}}
 {id=WavePlayer_2, properties {soundWaveAssetPtr=&/BarnTowel_Metal_Land_02}}
 {id=WavePlayer_3, properties {soundWaveAssetPtr=&/BarnTowel_Metal_Land_01}}
 {id=WavePlayer_4, properties {soundWaveAssetPtr=&/BarnTowell_Drop}}
]
properties {
 attenuationSettings=&Fort_Weapon_Drop_Att
 maxDistance=3000
 soundClassObject=&Fort_Gadgets
 volumeMultiplier=0.6
}
