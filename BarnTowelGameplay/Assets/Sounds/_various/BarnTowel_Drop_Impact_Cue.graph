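# Sound cue for the metal drop impact of the BarnTowel item.
# A Random node picks one of four equally weighted impact wave variations and feeds a Modulator that slightly lowers pitch at random to reduce repetition.
# The cue plays at reduced volume under the Fort_Gadgets sound class, with weapon-drop attenuation for spatial falloff.
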SoundCue /BarnTowelGameplay/Assets/Sounds/_various/BarnTowel_Drop_Impact_Cue
root=Modulator_0
nodes [
 {id=Modulator_0, children [Random_0], properties {pitchMax=1, pitchMin=0.9}}
 {id=Random_0, children [WavePlayer_0,WavePlayer_1,WavePlayer_2,WavePlayer_3], properties {weights [1,1,1,1]}}
 {id=WavePlayer_0, properties {soundWaveAssetPtr=&/BarnTowel_Metal_Drop_Impact_01}}
 {id=WavePlayer_1, properties {soundWaveAssetPtr=&/BarnTowel_Metal_Drop_Impact_02}}
 {id=WavePlayer_2, properties {soundWaveAssetPtr=&/BarnTowel_Metal_Drop_Impact_03}}
 {id=WavePlayer_3, properties {soundWaveAssetPtr=&/BarnTowel_Metal_Drop_Impact_04}}
]
properties {
 attenuationSettings=&Fort_Weapon_Drop_Att
 maxDistance=3000
 soundClassObject=&Fort_Gadgets
 volumeMultiplier=0.6
}
