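# BarnTowel_Equip_Cue is the Sound Cue for equipping the BarnTowel gadget.
# A Random node picks evenly between two equip waves, then a Modulator varies the pitch slightly upward.
# It plays at reduced volume through the Fort_Gadgets sound class with weapon reload attenuation, and is primed on load.
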SoundCue /BarnTowelGameplay/Assets/Sounds/_various/BarnTowel_Equip_Cue
root=Modulator_0
nodes [
 {id=Modulator_0, children [Random_0], properties {pitchMax=1.1, pitchMin=1}}
 {id=Random_0, children [WavePlayer_1,WavePlayer_0], properties {weights [1,1]}}
 {id=WavePlayer_0, properties {soundWaveAssetPtr=&/BarnTowel_Equip_01}}
 {id=WavePlayer_1, properties {soundWaveAssetPtr=&/BarnTowel_Equip}}
]
properties {
 attenuationSettings=&Fort_Weapon_Reload_Att
 bPrimeOnLoad=true
 maxDistance=2250
 soundClassObject=&Fort_Gadgets
 volumeMultiplier=0.5
}
