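# Sound cue for the BarnTowel explosive impact, branching on the islocal parameter between local and remote mixes.
# The local branch and remote close layer share metal-impact and base sounds, layered with secondary-fire audio and randomized close explosions.
# The remote branch adds medium and distant fire-plus-explosion layers, each band using its own attenuation, with sound class overrides separating 1P, 3P and distant.
# A concurrency override allows one instance, stopping the oldest.
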SoundCue /BarnTowelGameplay/Assets/Sounds/_various/BarnTowel_Explosive_Impact_Cue
root=Branch_0
nodes [
 {
  id=Attenuation_0
  children [Mixer_3]
  properties {
   attenuationSettings=&Fort_Weapon_Projectile_Close_Athena_Att
   attenuationOverrides {stereoSpread=1000, attenuationShapeExtents {x=1000}}
  }
 }
 {
  id=Attenuation_2
  children [Mixer_7]
  properties {attenuationSettings=&Fort_Weapon_Projectile_Distant_Mono_Athena_Att}
 }
 {
  id=Attenuation_29
  children [Mixer_6]
  properties {attenuationSettings=&Fort_Weapon_Projectile_Medium_Mono_Athena_Att}
 }
 {id=Attenuation_4, children [Mixer_2]}
 {id=Branch_0, children [Mixer_9,Mixer_1,Mixer_1], properties {boolParameterName=islocal}}
 {id=Mixer_0, children [SoundClass_5, SoundClass_30], properties {inputVolume [1.25,0.75]}}
 {id=Mixer_1, children [Mixer_0,SoundClass_0], properties {inputVolume [1,1]}}
 {id=Mixer_2, children [Modulator_2,Mixer_4], properties {inputVolume [1,1]}}
 {id=Mixer_3, children [Modulator_2,Mixer_5], properties {inputVolume [1,1]}}
 {id=Mixer_4, children [Modulator_4,Random_1], properties {inputVolume [1.25,0.5]}}
 {id=Mixer_5, children [Modulator_1,Random_1], properties {inputVolume [1.25,0.5]}}
 {id=Mixer_6, children [Modulator_3,Random_3], properties {inputVolume [1,0.75]}}
 {id=Mixer_7, children [Modulator_0,Random_2], properties {inputVolume [1,1]}}
 {id=Mixer_8, children [Random_0, WavePlayer_15], properties {inputVolume [0.55,1]}}
 {id=Mixer_9, children [SoundClass_2], properties {inputVolume [1.6]}}
 {id=Modulator_0, children [WavePlayer_16]}
 {id=Modulator_1, children [WavePlayer_0]}
 {id=Modulator_2, children [Mixer_8], properties {volumeMax=1.3, volumeMin=1.3}}
 {id=Modulator_3, children [WavePlayer_17]}
 {id=Modulator_4, children [WavePlayer_18], properties {volumeMax=1.6, volumeMin=1.6}}
 {id=Random_0, children [WavePlayer_12, WavePlayer_13, WavePlayer_14], properties {weights [1,1,1]}}
 {id=Random_1, children [WavePlayer_19, WavePlayer_20, WavePlayer_21, WavePlayer_22], properties {weights [1,1,1,1]}}
 {id=Random_2, children [WavePlayer_23, WavePlayer_24, WavePlayer_25, WavePlayer_26], properties {weights [1,1,1,1]}}
 {id=Random_3, children [WavePlayer_27, WavePlayer_28, WavePlayer_29, WavePlayer_30], properties {weights [1,1,1,1]}}
 {
  id=SoundClass_0
  children [Attenuation_2]
  properties {soundClassOverride=&Fort_WeaponRanged_Fire_Distant}
 }
 {
  id=SoundClass_2
  children [Attenuation_4]
  properties {soundClassOverride=&Fort_WeaponRanged_Fire_1P_PSM}
 }
 {
  id=SoundClass_30
  children [Attenuation_29]
  properties {soundClassOverride=&Fort_WeaponRanged_Fire_Distant}
 }
 {
  id=SoundClass_5
  children [Attenuation_0]
  properties {soundClassOverride=&Fort_WeaponRanged_Fire_3P_PSM}
 }
 {id=WavePlayer_0, properties {soundWaveAssetPtr=&/BarnTowel_Secondary_Fire_Close_01}}
 {id=WavePlayer_12, properties {soundWaveAssetPtr=&/BarnTowel_Land_Impact_Metal_01}}
 {id=WavePlayer_13, properties {soundWaveAssetPtr=&/BarnTowel_Land_Impact_Metal_03}}
 {id=WavePlayer_14, properties {soundWaveAssetPtr=&/BarnTowel_Land_Impact_Metal_02}}
 {id=WavePlayer_15, properties {soundWaveAssetPtr=&/BarnTowel_ImpactBase}}
 {id=WavePlayer_16, properties {soundWaveAssetPtr=&/BarnTowel_Secondary_Fire_3P_Distant_01}}
 {id=WavePlayer_17, properties {soundWaveAssetPtr=&/BarnTowel_Secondary_Fire_3P_Medium_01}}
 {id=WavePlayer_18, properties {soundWaveAssetPtr=&BarnTowel_Secondary_Fire_Close_01}}
 {id=WavePlayer_19, properties {soundWaveAssetPtr=&/BarnTowel_Explosion_Close_01}}
 {id=WavePlayer_20, properties {soundWaveAssetPtr=&/BarnTowel_Explosion_Close_02}}
 {id=WavePlayer_21, properties {soundWaveAssetPtr=&/BarnTowel_Explosion_Close_03}}
 {id=WavePlayer_22, properties {soundWaveAssetPtr=&/BarnTowel_Explosion_Close_04}}
 {id=WavePlayer_23, properties {soundWaveAssetPtr=&/BarnTowel_Explosion_Distant_01}}
 {id=WavePlayer_24, properties {soundWaveAssetPtr=&/BarnTowel_Explosion_Distant_02}}
 {id=WavePlayer_25, properties {soundWaveAssetPtr=&/BarnTowel_Explosion_Distant_03}}
 {id=WavePlayer_26, properties {soundWaveAssetPtr=&/BarnTowel_Explosion_Distant_04}}
 {id=WavePlayer_27, properties {soundWaveAssetPtr=&/BarnTowel_Explosion_Medium_01}}
 {id=WavePlayer_28, properties {soundWaveAssetPtr=&/BarnTowel_Explosion_Medium_02}}
 {id=WavePlayer_29, properties {soundWaveAssetPtr=&/BarnTowel_Explosion_Medium_03}}
 {id=WavePlayer_30, properties {soundWaveAssetPtr=&/BarnTowel_Explosion_Medium_04}}
]
properties {
 attenuationSettings=/BarnTowelGameplay/Assets/Sounds/Settings/BarnTowel_InFlightAir_3P_Att
 bOverrideConcurrency=true
 maxDistance=2097152
 soundClassObject=/Game/Sounds/FortSoundClassesAndModes/Fort_Events_Weapons_Impact
 volumeMultiplier=0.65
 concurrencyOverrides {maxCount=1, resolutionRule=StopOldest}
 concurrencySet [&Player_Melee_Weapon_Conc]
}
