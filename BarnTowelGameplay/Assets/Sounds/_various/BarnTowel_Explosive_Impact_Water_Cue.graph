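# Sound cue for the water impact of a BarnTowel explosive, layering two splash sounds.
# A mixer combines one random "enter" splash with one random medium splash, each from three variants.
# An "islocal" branch routes this mix to a first-person weapon sound class, or to a third-person class with in-flight-air attenuation when false or unset.
# A concurrency override allows one instance, stopping the oldest.
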SoundCue /BarnTowelGameplay/Assets/Sounds/_various/BarnTowel_Explosive_Impact_Water_Cue
root=Branch_0
nodes [
 {
  id=Attenuation_1
  children [Mixer_0]
  properties {attenuationSettings=/BarnTowelGameplay/Assets/Sounds/Settings/BarnTowel_InFlightAir_3P_Att}
 }
 {id=Attenuation_4, children [Mixer_0]}
 {id=Branch_0, children [SoundClass_2,SoundClass_5,SoundClass_5], properties {boolParameterName=islocal}}
 {id=Mixer_0, children [Random_2,Random_1], properties {inputVolume [1,1]}}
 {id=Random_1, children [WavePlayer_3,WavePlayer_7,WavePlayer_8], properties {weights [1,1,1]}}
 {id=Random_2, children [WavePlayer_0,WavePlayer_1,WavePlayer_2], properties {weights [1,1,1]}}
 {
  id=SoundClass_2
  children [Attenuation_4]
  properties {soundClassOverride=&Fort_WeaponRanged_Fire_1P_PSM}
 }
 {
  id=SoundClass_5
  children [Attenuation_1]
  properties {soundClassOverride=&Fort_WeaponRanged_Fire_3P_PSM}
 }
 {id=WavePlayer_0, properties {soundWaveAssetPtr=&/BarnTowel_Water_Splash_Enter_01}}
 {id=WavePlayer_1, properties {soundWaveAssetPtr=&/BarnTowel_Water_Splash_Enter_02}}
 {id=WavePlayer_2, properties {soundWaveAssetPtr=&/BarnTowel_Water_Splash_Enter_03}}
 {id=WavePlayer_3, properties {soundWaveAssetPtr=&/BarnTowel_Water_Splash_Med_01}}
 {id=WavePlayer_7, properties {soundWaveAssetPtr=&/BarnTowel_Water_Splash_Med_02}}
 {id=WavePlayer_8, properties {soundWaveAssetPtr=&/BarnTowel_Water_Splash_Med_03}}
]
properties {
 attenuationSettings=&BarnTowel_InFlightAir_3P_Att
 bOverrideConcurrency=true
 maxDistance=2097152
 soundClassObject=&Fort_Gadgets
 volumeMultiplier=1.5
 concurrencyOverrides {maxCount=1, resolutionRule=StopOldest}
 concurrencySet [&Player_Melee_Weapon_Conc]
}
