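# Sound cue for the BarnTowel gadget running out of fuel, routed to the Fort_Gadgets sound class.
# A mixer layers two fuel-ending sets, a crash impact, two sputters, a tonal servo and a sputter burst, with random wave picks, pitch modulation and two fixed delays for variation and staggering.
# An islocal branch sits at the root, but all three outputs share the same third-person in-flight attenuation node, so every case sounds alike.
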SoundCue /BarnTowelGameplay/Assets/Sounds/_various/BarnTowel_Fuel_Out_Cue
root=Branch_0
nodes [
 {
  id=Attenuation_1
  children [Mixer_0]
  properties {attenuationSettings=/BarnTowelGameplay/Assets/Sounds/Settings/BarnTowel_InFlightAir_3P_Att}
 }
 {id=Branch_0, children [Attenuation_1, Attenuation_1, Attenuation_1], properties {boolParameterName=islocal}}
 {id=Delay_1, children [Modulator_6], properties {delayMax=0.5, delayMin=0.5}}
 {id=Delay_2, children [Modulator_1], properties {delayMax=0.6, delayMin=0.6}}
 {
  id=Mixer_0
  children [Modulator_2,Delay_2,Random_2,Modulator_5,Delay_1,Modulator_4,Random_4]
  properties {inputVolume [0.5,1,1,0.25,0.15,0.9,2]}
 }
 {id=Modulator_1, children [Random_1], properties {pitchMax=1.1, pitchMin=1}}
 {id=Modulator_2, children [Random_0], properties {pitchMax=1.1, pitchMin=1}}
 {id=Modulator_4, children [Random_5], properties {pitchMax=0.95, pitchMin=0.85}}
 {id=Modulator_5, children [WavePlayer_12], properties {pitchMax=1.2, pitchMin=0.98}}
 {id=Modulator_6, children [WavePlayer_13], properties {pitchMax=1.1, pitchMin=0.98}}
 {id=Random_0, children [WavePlayer_2], properties {weights [1]}}
 {id=Random_1, children [WavePlayer_3,WavePlayer_4,WavePlayer_5], properties {weights [1,1,1]}}
 {id=Random_2, children [WavePlayer_7,WavePlayer_8], properties {weights [1,1]}}
 {id=Random_4, children [WavePlayer_16, WavePlayer_17, WavePlayer_18, WavePlayer_19, WavePlayer_20], properties {weights [1,1,1,1,1]}}
 {id=Random_5, children [WavePlayer_14, WavePlayer_15], properties {weights [1,1]}}
 {id=WavePlayer_12, properties {soundWaveAssetPtr=&/BarnTowel_Fuelout_Sputter_01}}
 {id=WavePlayer_13, properties {soundWaveAssetPtr=&/BarnTowel_Fuelout_Sputter_02}}
 {id=WavePlayer_14, properties {soundWaveAssetPtr=&/BarnTowel_Fuelout_Tonalservo_01}}
 {id=WavePlayer_15, properties {soundWaveAssetPtr=&/BarnTowel_Fuelout_Tonalservo_02}}
 {id=WavePlayer_16, properties {soundWaveAssetPtr=&/BarnTowel_Sputter_Burst_01}}
 {id=WavePlayer_17, properties {soundWaveAssetPtr=&/BarnTowel_Sputter_Burst_03}}
 {id=WavePlayer_18, properties {soundWaveAssetPtr=&/BarnTowel_Sputter_Burst_04}}
 {id=WavePlayer_19, properties {soundWaveAssetPtr=&/BarnTowel_Sputter_Burst_05}}
 {id=WavePlayer_2, properties {soundWaveAssetPtr=&/BarnTowel_Fuel_Ending_A_02}}
 {id=WavePlayer_20, properties {soundWaveAssetPtr=&/BarnTowel_Sputter_Burst_06}}
 {id=WavePlayer_3, properties {soundWaveAssetPtr=&/BarnTowel_Fuel_Ending_B_01}}
 {id=WavePlayer_4, properties {soundWaveAssetPtr=&/BarnTowel_Fuel_Ending_B_02}}
 {id=WavePlayer_5, properties {soundWaveAssetPtr=&/BarnTowel_Fuel_Ending_B_03}}
 {id=WavePlayer_7, properties {soundWaveAssetPtr=&/BarnTowel_ForceCrash_Impt_01}}
 {id=WavePlayer_8, properties {soundWaveAssetPtr=&/BarnTowel_ForceCrash_Impt_02}}
]
properties {
 attenuationSettings=&BarnTowel_InFlightAir_3P_Att
 bOverrideAttenuation=true
 maxDistance=2097152
 soundClassObject=&Fort_Gadgets
}
