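# Sound cue for the default wall-break impact of the BarnTowel, classed as a melee weapon sound.
# A random node picks one of four equally weighted impact waves, and a modulator above it slightly varies pitch and volume.
# The cue lowers overall pitch and volume, uses the BarnTowel fire attenuation asset, and overrides concurrency to cap simultaneous plays per owner.
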SoundCue /BarnTowelGameplay/Assets/Sounds/_various/BarnTowel_Impact_BreakWall_Default_Cue
root=Modulator_3
nodes [
 {id=Modulator_3, children [Random_4], properties {pitchMax=1.1, pitchMin=0.9, volumeMax=1.1, volumeMin=0.9}}
 {id=Random_4, children [WavePlayer_0,WavePlayer_1,WavePlayer_2,WavePlayer_3], properties {weights [1,1,1,1]}}
 {id=WavePlayer_0, properties {soundWaveAssetPtr=&/BarnTowel_Impact_BreakWall_Default_04}}
 {id=WavePlayer_1, properties {soundWaveAssetPtr=&/BarnTowel_Impact_BreakWall_Default_03}}
 {id=WavePlayer_2, properties {soundWaveAssetPtr=&/BarnTowel_Impact_BreakWall_Default_02}}
 {id=WavePlayer_3, properties {soundWaveAssetPtr=&/BarnTowel_Impact_BreakWall_Default_01}}
]
properties {
 attenuationSettings=/BarnTowelGameplay/Assets/Sounds/Settings/BarnTowel_Fire_Att
 bOverrideConcurrency=true
 maxDistance=12000
 pitchMultiplier=0.8
 soundClassObject=&Fort_WeaponMelee
 volumeMultiplier=0.4
 concurrencyOverrides {bLimitToOwner=true, maxCount=2}
}
