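# Sound cue for the start of the BarnTowel launch boost, with one instance limited per owner.
# Branches on islocal, then isenemy, play the shared mix at full volume, except non-local non-enemies, which get half volume.
# The mix layers randomized close, medium and distant variations through sound class overrides.
# The distant layer has its own projectile attenuation, separate from the cue's in-flight third-person attenuation.
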SoundCue /BarnTowelGameplay/Assets/Sounds/_various/BarnTowel_Launch_Start_Cue
root=Branch_2
nodes [
 {
  id=Attenuation_5
  children [Modulator_3]
  properties {attenuationSettings=&Fort_Weapon_Projectile_Distant_Mono_Athena_Att}
 }
 {id=Branch_2, children [Modulator_0,Branch_3,Modulator_0], properties {boolParameterName=islocal}}
 {id=Branch_3, children [Modulator_5,Modulator_4,Modulator_5], properties {boolParameterName=isenemy}}
 {id=Mixer_1, children [SoundClass_2,Mixer_4], properties {inputVolume [1.2,0.75]}}
 {id=Mixer_4, children [SoundClass_3,SoundClass_4], properties {inputVolume [0.8,0.6]}}
 {id=Modulator_0, children [Mixer_1], properties {pitchMax=1, pitchMin=1, volumeMax=1, volumeMin=1}}
 {id=Modulator_1, children [Random_2]}
 {id=Modulator_2, children [Random_4]}
 {id=Modulator_3, children [Random_5]}
 {id=Modulator_4, children [Mixer_1], properties {pitchMax=1, pitchMin=1, volumeMax=0.5, volumeMin=0.5}}
 {id=Modulator_5, children [Mixer_1], properties {pitchMax=1, pitchMin=1, volumeMax=1, volumeMin=1}}
 {id=Random_2, children [WavePlayer_0,WavePlayer_1,WavePlayer_2,WavePlayer_3,WavePlayer_4,WavePlayer_5], properties {weights [1,1,1,1,1,1]}}
 {
  id=Random_4
  children [WavePlayer_30, WavePlayer_31, WavePlayer_32, WavePlayer_33, WavePlayer_34, WavePlayer_35]
  properties {weights [1,1,1,1,1,1]}
 }
 {
  id=Random_5
  children [WavePlayer_6, WavePlayer_7, WavePlayer_8, WavePlayer_9, WavePlayer_28, WavePlayer_29]
  properties {weights [1,1,1,1,1,1]}
 }
 {id=SoundClass_2, children [Modulator_1], properties {soundClassOverride=&Fort_WeaponRanged_Fire_3P_PSM}}
 {
  id=SoundClass_3
  children [Modulator_2]
  properties {soundClassOverride=&Fort_WeaponRanged_Fire_Distant}
 }
 {
  id=SoundClass_4
  children [Attenuation_5]
  properties {soundClassOverride=&Fort_WeaponRanged_Fire_Distant}
 }
 {id=WavePlayer_0, properties {soundWaveAssetPtr=&/BarnTowel_Boost_Start_Close_01}}
 {id=WavePlayer_1, properties {soundWaveAssetPtr=&/BarnTowel_Boost_Start_Close_02}}
 {id=WavePlayer_2, properties {soundWaveAssetPtr=&/BarnTowel_Boost_Start_Close_03}}
 {id=WavePlayer_28, properties {soundWaveAssetPtr=&/BarnTowel_Boost_Start_Distant_05}}
 {id=WavePlayer_29, properties {soundWaveAssetPtr=&/BarnTowel_Boost_Start_Distant_06}}
 {id=WavePlayer_3, properties {soundWaveAssetPtr=&/BarnTowel_Boost_Start_Close_04}}
 {id=WavePlayer_30, properties {soundWaveAssetPtr=&/BarnTowel_Boost_Start_Medium_01}}
 {id=WavePlayer_31, properties {soundWaveAssetPtr=&/BarnTowel_Boost_Start_Medium_02}}
 {id=WavePlayer_32, properties {soundWaveAssetPtr=&/BarnTowel_Boost_Start_Medium_03}}
 {id=WavePlayer_33, properties {soundWaveAssetPtr=&/BarnTowel_Boost_Start_Medium_04}}
 {id=WavePlayer_34, properties {soundWaveAssetPtr=&/BarnTowel_Boost_Start_Medium_05}}
 {id=WavePlayer_35, properties {soundWaveAssetPtr=&/BarrnTowel_Boost_Start_Medium_06}}
 {id=WavePlayer_4, properties {soundWaveAssetPtr=&/BarnTowel_Boost_Start_Close_05}}
 {id=WavePlayer_5, properties {soundWaveAssetPtr=&/BarnTowel_Boost_Start_Close_06}}
 {id=WavePlayer_6, properties {soundWaveAssetPtr=&/BarnTowel_Boost_Start_Distant_01}}
 {id=WavePlayer_7, properties {soundWaveAssetPtr=&/BarnTowel_Boost_Start_Distant_02}}
 {id=WavePlayer_8, properties {soundWaveAssetPtr=&/BarnTowel_Boost_Start_Distant_03}}
 {id=WavePlayer_9, properties {soundWaveAssetPtr=&/BarnTowel_Boost_Start_Distant_04}}
]
properties {
 attenuationSettings=/BarnTowelGameplay/Assets/Sounds/Settings/BarnTowel_InFlightAir_3P_Att
 maxDistance=2097152
 soundClassObject=&Fort_Gadgets
 volumeMultiplier=1.15
 concurrencyOverrides {bLimitToOwner=true, maxCount=1}
 concurrencySet [&Weapon_Gunfire_Global_Conc, &Weapon_Gunfire_Local_Conc]
}
