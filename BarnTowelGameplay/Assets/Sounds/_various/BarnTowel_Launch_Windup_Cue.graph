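# Sound cue for the BarnTowel launch windup, using a third-person in-flight attenuation and gunfire concurrency sets.
# The main branch, classed as third-person weapon fire, mixes random charge-up and explosion variants with delayed turbine waves and a delayed flare-up plus air-blast layer.
# Two quieter branches override to the distant weapon-fire class and play random medium charge-ups, one pitched up and one pitched down.
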SoundCue /BarnTowelGameplay/Assets/Sounds/_various/BarnTowel_Launch_Windup_Cue
root=Mixer_1
nodes [
 {id=Delay_0, children [Mixer_2], properties {delayMax=0.8, delayMin=0.8}}
 {id=Delay_1, children [WavePlayer_43], properties {delayMax=0.5, delayMin=0.5}}
 {id=Delay_2, children [WavePlayer_42], properties {delayMax=0.05, delayMin=0.05}}
 {id=Mixer_0, children [Random_2,Delay_0,Random_1,Delay_2,Delay_1], properties {inputVolume [1.3,0.8,1.3,0.8,0.25]}}
 {id=Mixer_1, children [SoundClass_2,SoundClass_3,SoundClass_0], properties {inputVolume [1,0.6,0.4]}}
 {id=Mixer_2, children [Random_0,WavePlayer_7], properties {inputVolume [1,0.7]}}
 {id=Modulator_0, children [Random_5], properties {pitchMax=0.95, pitchMin=0.85, volumeMax=2, volumeMin=2}}
 {id=Modulator_8, children [Mixer_0], properties {volumeMax=1.1, volumeMin=1.1}}
 {id=Modulator_9, children [Random_4], properties {pitchMax=1.2, pitchMin=1.1, volumeMax=2, volumeMin=2}}
 {id=Random_0, children [WavePlayer_4,WavePlayer_5,WavePlayer_6], properties {weights [1,1,1]}}
 {id=Random_1, children [WavePlayer_8, WavePlayer_9, WavePlayer_10], properties {weights [1,1,1]}}
 {id=Random_2, children [WavePlayer_0,WavePlayer_1,WavePlayer_2,WavePlayer_3], properties {weights [1,1,1,1]}}
 {id=Random_4, children [WavePlayer_11, WavePlayer_12, WavePlayer_13, WavePlayer_14], properties {weights [1,1,1,1]}}
 {id=Random_5, children [WavePlayer_15, WavePlayer_16, WavePlayer_17, WavePlayer_18], properties {weights [1,1,1,1]}}
 {
  id=SoundClass_0
  children [Modulator_0]
  properties {soundClassOverride=&Fort_WeaponRanged_Fire_Distant}
 }
 {id=SoundClass_2, children [Modulator_8], properties {soundClassOverride=&Fort_WeaponRanged_Fire_3P_PSM}}
 {
  id=SoundClass_3
  children [Modulator_9]
  properties {soundClassOverride=&Fort_WeaponRanged_Fire_Distant}
 }
 {id=WavePlayer_0, properties {soundWaveAssetPtr=&/BarnTowel_ChargeUp_01}}
 {id=WavePlayer_1, properties {soundWaveAssetPtr=&/BarnTowel_ChargeUp_02}}
 {id=WavePlayer_10, properties {soundWaveAssetPtr=&/BarnTowel_ChargeUp_AOE_Explosion_03}}
 {id=WavePlayer_11, properties {soundWaveAssetPtr=&/BarnTowel_ChargeUp_Medium_01}}
 {id=WavePlayer_12, properties {soundWaveAssetPtr=&/BarnTowel_ChargeUp_Medium_02}}
 {id=WavePlayer_13, properties {soundWaveAssetPtr=&/BarnTowel_ChargeUp_Medium_03}}
 {id=WavePlayer_14, properties {soundWaveAssetPtr=&/BarnTowel_ChargeUp_Medium_04}}
 {id=WavePlayer_15, properties {soundWaveAssetPtr=&BarnTowel_ChargeUp_Medium_01}}
 {id=WavePlayer_16, properties {soundWaveAssetPtr=&BarnTowel_ChargeUp_Medium_02}}
 {id=WavePlayer_17, properties {soundWaveAssetPtr=&BarnTowel_ChargeUp_Medium_03}}
 {id=WavePlayer_18, properties {soundWaveAssetPtr=&BarnTowel_ChargeUp_Medium_04}}
 {id=WavePlayer_2, properties {soundWaveAssetPtr=&/BarnTowel_ChargeUp_03}}
 {id=WavePlayer_3, properties {soundWaveAssetPtr=&/BarnTowel_ChargeUp_04}}
 {id=WavePlayer_4, properties {soundWaveAssetPtr=&/BarnTowel_ChargeUp_Layer_Flareup_01}}
 {id=WavePlayer_42, properties {soundWaveAssetPtr=&/BarnTowel_Launch_Windup_Turbine_01}}
 {id=WavePlayer_43, properties {soundWaveAssetPtr=&/BarnTowel_Launch_Windup_Turbine_02}}
 {id=WavePlayer_5, properties {soundWaveAssetPtr=&/BarnTowel_ChargeUp_Layer_Flareup_02}}
 {id=WavePlayer_6, properties {soundWaveAssetPtr=&/BarnTowel_ChargeUp_Layer_Flareup_03}}
 {id=WavePlayer_7, properties {soundWaveAssetPtr=&/BarnTowel_ChargeUp_Blast_Layer_Air_01}}
 {id=WavePlayer_8, properties {soundWaveAssetPtr=&/BarnTowel_ChargeUp_AOE_Explosion_01}}
 {id=WavePlayer_9, properties {soundWaveAssetPtr=&/BarnTowel_ChargeUp_AOE_Explosion_02}}
]
properties {
 attenuationSettings=/BarnTowelGameplay/Assets/Sounds/Settings/BarnTowel_InFlightAir_3P_Att
 maxDistance=2097152
 soundClassObject=&Fort_Gadgets
 volumeMultiplier=0.65
 concurrencyOverrides {bLimitToOwner=true, maxCount=1}
 concurrencySet [&Weapon_Gunfire_Global_Conc, &Weapon_Gunfire_Local_Conc]
}
