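# Sound cue for the BarnTowel metal landing sound, classed under Fort_Gadgets with the BarnTowel ground attenuation.
# A Random node picks one of three equally weighted land waves.
# A Branch on the "islocal" bool plays that pick directly when true, and routes it through a half-volume Mixer when false or unset, so non-local landings are quieter.
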SoundCue /BarnTowelGameplay/Assets/Sounds/_various/BarnTowel_Metal_Land_Cue
root=Branch_0
nodes [
 {id=Branch_0, children [Random_0,Mixer_1,Mixer_1], properties {boolParameterName=islocal}}
 {id=Mixer_1, children [Random_0], properties {inputVolume [0.5]}}
 {id=Random_0, children [WavePlayer_0,WavePlayer_1,WavePlayer_2], properties {weights [1,1,1]}}
 {id=WavePlayer_0, properties {soundWaveAssetPtr=&/BarnTowel_Metal_Land_01}}
 {id=WavePlayer_1, properties {soundWaveAssetPtr=&/BarnTowel_Metal_Land_02}}
 {id=WavePlayer_2, properties {soundWaveAssetPtr=&/BarnTowel_Metal_Land_03}}
]
properties {
 attenuationSettings=/BarnTowelGameplay/Assets/Sounds/Settings/BarnTowel_Ground_Att
 maxDistance=3900
 soundClassObject=&Fort_Gadgets
 volumeMultiplier=0.6
}
