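# Sound cue for picking up the BarnTowel item in the BarnTowelGameplay plugin.
# A single BarnTowel_Pickup wave player feeds a root Modulator that applies slight random pitch variation on each play.
# It plays at reduced volume in the Fort_Gadgets sound class, spatialized through the shared weapon pickup attenuation settings.
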SoundCue /BarnTowelGameplay/Assets/Sounds/_various/BarnTowel_Pickup_Cue
root=Modulator_0
nodes [
 {id=Modulator_0, children [WavePlayer_1], properties {pitchMax=1.02, pitchMin=0.98}}
 {id=WavePlayer_1, properties {soundWaveAssetPtr=&/BarnTowel_Pickup}}
]
properties {
 attenuationSettings=&Fort_Weapon_Pickup_Att
 maxDistance=3500
 soundClassObject=&Fort_Gadgets
 volumeMultiplier=0.4
}
