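# Sound cue for the BarnTowel gadget's thrust-off landing, routed to the Fort_Gadgets sound class with ground attenuation.
# A mixer layers a random thrust-off land wave with a random force-crash impact wave.
# A branch on the "islocal" bool plays that mix directly when true, or through a second mixer at half volume when false or unset.
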SoundCue /BarnTowelGameplay/Assets/Sounds/_various/BarnTowel_ThrustOff_Land_Cue
root=Branch_0
nodes [
 {id=Branch_0, children [Mixer_0,Mixer_1,Mixer_1], properties {boolParameterName=islocal}}
 {id=Mixer_0, children [Random_0,Random_1], properties {inputVolume [1,1]}}
 {id=Mixer_1, children [Mixer_0], properties {inputVolume [0.5]}}
 {id=Random_0, children [WavePlayer_0,WavePlayer_1,WavePlayer_2], properties {weights [1,1,1]}}
 {id=Random_1, children [WavePlayer_5,WavePlayer_6], properties {weights [1,1]}}
 {id=WavePlayer_0, properties {soundWaveAssetPtr=&/BarnTowel_Thrust_Off_Land_01}}
 {id=WavePlayer_1, properties {soundWaveAssetPtr=&/BarnTowel_Thrust_Off_Land_02}}
 {id=WavePlayer_2, properties {soundWaveAssetPtr=&/BarnTowel_Thrust_Off_Land_03}}
 {id=WavePlayer_5, properties {soundWaveAssetPtr=&/BarnTowel_ForceCrash_Impt_01}}
 {id=WavePlayer_6, properties {soundWaveAssetPtr=&/BarnTowel_ForceCrash_Impt_02}}
]
properties {
 attenuationSettings=/BarnTowelGameplay/Assets/Sounds/Settings/BarnTowel_Ground_Att
 maxDistance=3900
 soundClassObject=&Fort_Gadgets
 volumeMultiplier=0.4
}
