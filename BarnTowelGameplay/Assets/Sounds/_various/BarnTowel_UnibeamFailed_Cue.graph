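# Sound cue for a failed BarnTowel Unibeam attempt, played on the Fort_UI sound class.
# A Mixer layers a modulated BarnTowel no-item sound with a quieter, slightly delayed stock "can't do" UI error ping.
# A concurrency override allows only a couple of simultaneous instances, stopping the oldest, with a short retrigger time to limit spamming.
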SoundCue /BarnTowelGameplay/Assets/Sounds/_various/BarnTowel_UnibeamFailed_Cue
root=Mixer_0
nodes [
 {id=Delay_0, children [WavePlayer_0], properties {delayMax=0.1, delayMin=0.1}}
 {id=Mixer_0, children [Modulator_0,Delay_0], properties {inputVolume [1.25,0.75]}}
 {id=Modulator_0, children [WavePlayer_1], properties {pitchMin=1}}
 {id=WavePlayer_0, properties {soundWaveAssetPtr=/Game/Sounds/Fort_UI_Sounds/Gameplay/UI_Error_Ping_Cant_Do_Default_01}}
 {id=WavePlayer_1, properties {soundWaveAssetPtr=&/BarnTowel_No_Item}}
]
properties {
 bOverrideConcurrency=true
 maxDistance=2097152
 soundClassObject=&Fort_UI
 volumeMultiplier=1
 concurrencyOverrides {maxCount=2, resolutionRule=StopOldest, retriggerTime=0.1}
}
