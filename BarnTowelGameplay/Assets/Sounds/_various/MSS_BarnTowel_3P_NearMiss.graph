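# Looping mono MetaSound source for the BarnTowel glider's third-person near-miss whoosh.
# When attenuation distance crosses a set threshold, a trigger compare checks the Speed input and, if it passes, a random forward whoosh plays through a wave player.
# Near-miss attenuation and a single-instance, prevent-new concurrency override limit playback; the air loop input is declared but unused.
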MetaSoundSource /BarnTowelGameplay/Assets/Sounds/_various/MSS_BarnTowel_3P_NearMiss
inputs [
 3P Bys {type=WaveAsset:Array}
 3PAirLoop {type=WaveAsset:Array}
 Speed {type=Float}
 UE.Attenuation.Distance {type=Float}
 UE.Source.OnPlay {type=Trigger}
 UE.Source.Orientation.Azimuth {type=Float}
 UE.Source.Orientation.Elevation {type=Float}
]
interfaces [UE.Attenuation:1.0, UE.OutputFormat.Mono:1.0, UE.Source.Orientation:1.0, UE.Source:1.1]
outputs ["UE.OutputFormat.Mono.Audio:0" {type=Audio}]
pages [Default {
  input_defaults {
   Speed=0
   UE.Attenuation.Distance=0
   UE.Source.OnPlay=null
   UE.Source.Orientation.Azimuth=0
   UE.Source.Orientation.Elevation=0
   3P Bys [
    &/BarnTowel_Forward_Whoosh_01
    &/BarnTowel_Forward_Whoosh_02
    &/BarnTowel_Forward_Whoosh_03
   ]
   3PAirLoop [&/BarnTowell_Thrust_Loop_01]
  }
  nodes [
   InitVariable.Float()
   ArrayRandomGet_0 = Array.Random Get:WaveAsset:Array(
    In Array=^3P Bys
    Next=TriggerCompare.Clamp:Float(A=^Speed, Compare=UE.TriggerOnThreshold:Float(In=^UE.Attenuation.Distance) {also {defaults {Threshold=1500, Type=1}}}.Out) {also {defaults {B=900, Type=3}}}.True
   )
   WavePlayer_0 = UE.Wave Player:Mono(Play=@ArrayRandomGet_0.On Next, Wave Asset=@ArrayRandomGet_0.Value)
  ]
  links [{dst="@UE.OutputFormat.Mono.Audio:0", src=WavePlayer_0.Out Mono}]
 }]
properties {
 assetImportData=&self:AssetImportData
 attenuationSettings=/BarnTowelGameplay/Assets/Sounds/Settings/BarnTowel_NearMiss_Att
 bLooping=true
 bOverrideConcurrency=true
 duration=10000
 maxDistance=2200
 pitch=1.2
 sampleRate=48000
 soundClassObject=&Fort_Gliders_InAir
 volume=0.75
 concurrencyOverrides {maxCount=1, resolutionRule=PreventNew, retriggerTime=5}
}
