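# MetaSound source for the BarnTowel glider's looping first-person in-flight air sound in 5.1.
# Speed, with an ascending override, crossfades low-speed thrust against high-speed loops, while pitch and roll drive thrust and sideways bank layers.
# Falling energy reshapes these layers and brings in a pitch-shifted sputter, released at empty.
# Leaving hover fires a random whoosh, and the final mix passes through wind-up muffling, speed-based EQ and a fade-in envelope.
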MetaSoundSource /BarnTowelGameplay/Assets/Sounds/_various/MSS_BarnTowel_FlyingAir_1P
inputs [
 EnergyRemaining {type=Float}
 Glider.LowSpeed {type=WaveAsset:Array}
 Glider.SidewaysLoop {type=WaveAsset:Array}
 Glider.SidewaysOneshots {type=WaveAsset:Array}
 Glider.StartFlyingOneShots {type=WaveAsset:Array}
 isAscending {type=Bool}
 isHovering {type=Bool}
 LowEnergy.Loop {type=WaveAsset:Array}
 Pitch {type=Float}
 Roll {type=Float}
 Speed {type=Float}
 SpeedThreshold.Loop {type=WaveAsset:Array}
 SpeedThreshold.SweetenerLoop {type=WaveAsset:Array}
 UE.Attenuation.Distance {type=Float}
 UE.Source.OnPlay {type=Trigger}
 WindUp {type=Bool}
]
interfaces [UE.Attenuation:1.0, UE.OutputFormat.5dot1:1.0, UE.Source:1.1]
outputs [
 "UE.OutputFormat.5dot1.Audio:0" {type=Audio}
 "UE.OutputFormat.5dot1.Audio:1" {type=Audio}
 "UE.OutputFormat.5dot1.Audio:2" {type=Audio}
 "UE.OutputFormat.5dot1.Audio:3" {type=Audio}
 "UE.OutputFormat.5dot1.Audio:4" {type=Audio}
 "UE.OutputFormat.5dot1.Audio:5" {type=Audio}
]
pages [Default {
  input_defaults {
   EnergyRemaining=0
   Pitch=0
   Roll=0.852
   Speed=1800
   UE.Attenuation.Distance=0
   UE.Source.OnPlay=null
   WindUp=null
   isAscending=null
   isHovering="<MetaSoundBuilderResult.SUCCEEDED: 0>"
   Glider.LowSpeed [&/BarnTowell_Thrust_Loop_01]
   Glider.SidewaysLoop [&/BarnTowel_Bank_Loop_01, &/BarnTowel_Bank_Loop_02]
   Glider.SidewaysOneshots [
    &/BarnTowel_Rattle_05
    &/BarnTowel_Rattle_04
    &/BarnTowel_Rattle_03
    &/BarnTowel_Rattle_02
    &/BarnTowel_Rattle_01
   ]
   Glider.StartFlyingOneShots [
    &/BarnTowel_Forward_Whoosh_03
    &/BarnTowel_Forward_Whoosh_02
    &/BarnTowel_Forward_Whoosh_01
   ]
   LowEnergy.Loop [&/BarnTowel_SputterLoop_01, &/BarnTowel_SputterLoopLayer_01]
   SpeedThreshold.Loop [/Game/Sounds/Fort_Gadgets/Athena/Bicycle/Glider_Bicycle_ThrusterLoop_Forwards]
   SpeedThreshold.SweetenerLoop [&/BarnTowel_HighSpeed_lp_01]
  }
  nodes [
   InitVariable.Float()
   InitVariable.Float()
   InitVariable.Float()
   TriggerOnValueChange_1 = UE.Trigger On Value Change:Bool(Value=^WindUp)
   TriggerAnyTriggerAny2_0 = "TriggerAny.Trigger Any (2)"(In 0=^UE.Source.OnPlay, In 1=@TriggerOnValueChange_1.Trigger)
   n1 = &MSP_Branch(Value=UE.Reroute(Value=^WindUp).Value, In=@TriggerAnyTriggerAny2_0.out)
   TriggerRouteTriggerRouteFloat2_4 = "TriggerRoute.Trigger Route (Float, 2):Float"(Set 1=@n1.FALSE, Set 0=@n1.TRUE) {also {defaults {Value 0=1, Value 1=0}}}
   n2 = MapRange.MapRange:Float(In=AbsoluteValue.Abs:Float(Input=^Roll).Value) {also {defaults {In Range B=3}}}
   TriggerOnValueChange_4 = UE.Trigger On Value Change:Float(Value=^Roll)
   n3 = VariableDeferredAccessor.Float(Variable=InitVariable.Float().Variable)
   n4 = TriggerCompare.Clamp:Float(B=@n2.Out Value, Compare=@TriggerOnValueChange_4.Trigger, A=@n3.Value) {also {defaults {Type=3}}}
   TriggerRouteTriggerRouteFloat2_5 = "TriggerRoute.Trigger Route (Float, 2):Float"(Set 1=@n4.FALSE, Set 0=@n4.TRUE) {also {defaults {Value 0=0.75, Value 1=0.1}}}
   n5 = UE.ConversionFloatToTime(In=@TriggerRouteTriggerRouteFloat2_5.Value)
   n6 = TriggerCompare.Clamp:Float(A=^Speed, Compare=@TriggerOnValueChange_4.Trigger) {also {defaults {B=800, Type=3}}}
   TriggerRouteTriggerRouteFloat2_0 = "TriggerRoute.Trigger Route (Float, 2):Float"(Value 0=@n2.Out Value, Set 1=@n6.FALSE, Set 0=@n6.TRUE) {also {defaults {Value 1=0}}}
   n7 = VariableAccessor.Float(Variable=VariableMutator.Float(Value=UE.InterpTo:Audio(Interp Time=@n5.out, Target=@TriggerRouteTriggerRouteFloat2_0.Value).Value, Variable=@n3.Variable).Variable)
   n8 = /FortMetasound/Cosmetics/Glider/Glider_ThrustLoop_Direction(Loop=^Glider.SidewaysLoop, Oneshots=^Glider.SidewaysOneshots, On Play=^UE.Source.OnPlay, Value=@n7.Value) {also {defaults {Interp Time=0, OneShotVolume=0.5, VolumeMin=0}}}
   n9 = VariableAccessor.Float(Variable=@n7.Variable)
   TriggerOnValueChange_5 = UE.Trigger On Value Change:Float(Value=@n9.Value)
   TriggerOnValueChange_7 = UE.Trigger On Value Change:Float(Value=^Pitch)
   TriggerAnyTriggerAny3_0 = "TriggerAny.Trigger Any (3)"(In 0=^UE.Source.OnPlay, In 2=@TriggerOnValueChange_5.Trigger, In 1=@TriggerOnValueChange_7.Trigger)
   n10 = TriggerCompare.Clamp:Float(A=^Speed, Compare=@TriggerOnValueChange_7.Trigger) {also {defaults {B=800, Type=3}}}
   TriggerRouteTriggerRouteFloat2_2 = "TriggerRoute.Trigger Route (Float, 2):Float"(Value 0=^Pitch, Value 1=Clamp.Clamp:Float(In=^Pitch) {also {defaults {Max=0.25}}}.Value, Set 1=@n10.FALSE, Set 0=@n10.TRUE)
   n11 = VariableAccessor.Float(Variable=VariableMutator.Float(Variable=InitVariable.Float().Variable, Value=@TriggerRouteTriggerRouteFloat2_2.Value).Variable)
   n12 = TriggerCompare.Clamp:Float(
    B=UE.Reroute(Value=UE.Reroute(Value=@n9.Value).Value).Value
    Compare=@TriggerAnyTriggerAny3_0.out
    A=VariableAccessor.Float(Variable=@n11.Variable).Value
   ) {also {defaults {Type=3}}}
   TriggerRouteTriggerRouteFloat2_3 = "TriggerRoute.Trigger Route (Float, 2):Float"(Value 1=UE.Reroute(Value=UE.Reroute(Value=@n9.Value).Value).Value, Set 1=@n12.FALSE, Set 0=@n12.TRUE, Value 0=@n11.Value)
   n13 = VariableAccessor.Float(Variable=VariableMutator.Float(Variable=InitVariable.Float().Variable, Value=@TriggerRouteTriggerRouteFloat2_3.Value).Variable)
   n14 = &Glider_ThrustLoop_Direction(Loop=^Glider.LowSpeed, On Play=^UE.Source.OnPlay, Value=@n13.Value) {also {defaults {Interp Time=0.35, PitchMin=1, VolumeMax=2, VolumeMin=1.6, bIsBackwards="<MetaSoundBuilderResult.SUCCEEDED: 0>"}}}
   TriggerOnValueChange_3 = UE.Trigger On Value Change:Float(Value=^Speed)
   TriggerOnValueChange_0 = UE.Trigger On Value Change:Bool(Value=^isAscending)
   TriggerAnyTriggerAny2_1 = "TriggerAny.Trigger Any (2)"(In 0=^UE.Source.OnPlay, In 1=@TriggerOnValueChange_0.Trigger)
   n15 = TriggerCompare.Clamp:Bool(A=UE.Reroute(Value=^isAscending).Value, Compare=@TriggerAnyTriggerAny2_1.out) {also {defaults {B="<MetaSoundBuilderResult.SUCCEEDED: 0>"}}}
   TriggerRouteTriggerRouteFloat2_1 = "TriggerRoute.Trigger Route (Float, 2):Float"(Value 1=Clamp.Clamp:Float(In=^Speed) {also {defaults {Max=1800}}}.Value, Set 1=@n15.FALSE, Set 0=@n15.TRUE) {also {defaults {Value 0=1100}}}
   n16 = VariableAccessor.Float(Variable=VariableMutator.Float(
     Variable=InitVariable.Float().Variable
     Value=/FortMetasound/Patches/MSP_ClampFloatChange(OnPlay=^UE.Source.OnPlay, ForceCalculation=@TriggerOnValueChange_3.Trigger, Value=@TriggerRouteTriggerRouteFloat2_1.Value) {also {defaults {Tolerance=50}}}.Output
    ).Variable)
   n17 = VariableAccessor.Float(Variable=@n16.Variable)
   n18 = /FortMetasound/Patches/MSP_Audio_EQCrossFade(In L=@n14.Out L, In R=@n14.Out R, Input=@n17.Value) {also {defaults {Interp Time=0.1, Max.DbAdjustment=6, Max.Value=1000, Min.DbAdjustment=-50, Min.Value=1800}}}
   TriggerOnValueChange_2 = UE.Trigger On Value Change:Bool(Value=^isHovering)
   ArrayRandomGet_0 = Array.Random Get:WaveAsset:Array(
    In Array=^Glider.StartFlyingOneShots
    Next=&MSP_Branch(Value=UE.Reroute(Value=^isHovering).Value, In=@TriggerOnValueChange_2.Trigger).FALSE
   ) {also {defaults {Weights []}}}
   WavePlayer_2 = UE.Wave Player:Stereo(Play=@ArrayRandomGet_0.On Next, Wave Asset=@ArrayRandomGet_0.Value) {also {version [1,1]}}
   AudioMixerAudioMixerStereo3_0 = "AudioMixer.Audio Mixer (Stereo, 3)"(In 2 L=@n8.Out L, In 2 R=@n8.Out R, In 1 L=@n18.Out L, In 1 R=@n18.Out R, In 0 L=@WavePlayer_2.Out Left, In 0 R=@WavePlayer_2.Out Right)
   n19 = &MSP_Audio_EQCrossFade(Input=^EnergyRemaining, In L=@AudioMixerAudioMixerStereo3_0.Out L, In R=@AudioMixerAudioMixerStereo3_0.Out R) {also {defaults {Max.Value=0.25, Min.DbAdjustment=-3, Min.HPF=200, Min.Value=0.05}}}
   ArrayRandomGet_1 = Array.Random Get:WaveAsset:Array(In Array=^SpeedThreshold.SweetenerLoop, Next=^UE.Source.OnPlay) {also {defaults {Weights []}}}
   WavePlayer_0 = UE.Wave Player:5dot1(
    Play=@ArrayRandomGet_1.On Next
    Wave Asset=@ArrayRandomGet_1.Value
    Pitch Shift=MapRange.MapRange:Float(In=^EnergyRemaining) {also {defaults {In Range A=0.15, In Range B=0, Out Range B=3}}}.Out Value
   ) {also {defaults {Loop="<MetaSoundBuilderResult.SUCCEEDED: 0>"}, version [1,1]}}
   ArrayRandomGet_2 = Array.Random Get:WaveAsset:Array(In Array=^SpeedThreshold.Loop, Next=^UE.Source.OnPlay)
   WavePlayer_1 = UE.Wave Player:Stereo(Play=@ArrayRandomGet_2.On Next, Wave Asset=@ArrayRandomGet_2.Value) {also {defaults {Loop="<MetaSoundBuilderResult.SUCCEEDED: 0>"}, version [1,1]}}
   n20 = /FortMetasound/Patches/Mixer_51_Two(
    "1.Volume(Lin)"=MapRange.MapRange:Float(In=^EnergyRemaining) {also {defaults {In Range A=0.25, In Range B=0, Out Range A=1.5, Out Range B=0}}}.Out Value
    2.InC=@WavePlayer_0.Out Front Center
    2.InL=@WavePlayer_0.Out Front Left
    2.InR=@WavePlayer_0.Out Front Right
    2.InLF=@WavePlayer_0.Out Low Frequency
    2.InSL=@WavePlayer_0.Out Side Left
    2.InSR=@WavePlayer_0.Out Side Right
    1.InL=@WavePlayer_1.Out Left
    1.InR=@WavePlayer_1.Out Right
   ) {also {defaults {"2.Volume(Lin)"=0.4}}}
   n21 = /FortMetasound/Patches/MSP_Audio_EQCrossFade_51(In C=@n20.OutC, In L=@n20.OutL, In LF=@n20.OutLF, In R=@n20.OutR, In SL=@n20.OutSL, In SR=@n20.OutSR, Input=@n16.Value) {also {defaults {Interp Time=0.1, Max.DbAdjustment=3, Max.Value=1600, Min.DbAdjustment=-50, Min.Value=1000}}}
   n22 = &MSP_Audio_EQCrossFade_51(Input=^EnergyRemaining, In C=@n21.Out C, In L=@n21.Out L, In LF=@n21.Out LF, In R=@n21.Out R, In SL=@n21.Out SL, In SR=@n21.Out SR) {also {defaults {Max.DbAdjustment=1, Max.Value=0.3, Min.DbAdjustment=6, Min.HPF=240, Min.Value=0.1}}}
   n23 = /FortMetasound/SoundCue/MSP_ContinuousModulator(Value In=^EnergyRemaining) {also {defaults {Input.Max=0.35, Output.Max=0, Output.Min=2}}}
   SemitonetoFrequencyMultiplier_0 = UE.Semitone to Frequency Multiplier:Float(Semitones=UE.Reroute(Value=@n23.Value Out).Value)
   SemitonetoFrequencyMultiplier_1 = UE.Semitone to Frequency Multiplier:Float(Semitones=UE.Add:Float(PrimaryOperand=@n23.Value Out) {also {defaults {AdditionalOperands=5}}}.out)
   n24 = &Glider_ThrustLoop_Direction(
    Loop=^LowEnergy.Loop
    On Play=^UE.Source.OnPlay
    PitchMin=@SemitonetoFrequencyMultiplier_0.Frequency Multiplier
    PitchMax=@SemitonetoFrequencyMultiplier_1.Frequency Multiplier
    Value=VariableAccessor.Float(Variable=@n13.Variable).Value
   ) {also {defaults {VolumeMax=1.3, VolumeMin=1, bIsBackwards="<MetaSoundBuilderResult.SUCCEEDED: 0>"}}}
   n25 = &MSP_Audio_EQCrossFade(Input=^EnergyRemaining, In L=@n24.Out L, In R=@n24.Out R) {also {defaults {Max.DbAdjustment=-50, Max.HPF=600, Max.Value=0.3, Min.DbAdjustment=-10, Min.Value=0.18}}}
   TriggerOnValueChange_6 = UE.Trigger On Value Change:Float(Value=^EnergyRemaining)
   ADSREnvelopeADSREnvelope_0 = ADSR Envelope.ADSR Envelope:Audio(
    Trigger Attack=^UE.Source.OnPlay
    Trigger Release=TriggerCompare.Clamp:Float(A=^EnergyRemaining, Compare=@TriggerOnValueChange_6.Trigger) {also {defaults {B=0}}}.TRUE
   ) {also {defaults {Attack Curve=2, Attack Time=0.25, Decay Time=0, Release Time=1.5, Sustain Level=1}}}
   n26 = VariableAccessor.Audio(Variable=VariableMutator.Audio(Value=@ADSREnvelopeADSREnvelope_0.Out Envelope, Variable=InitVariable.Audio().Variable).Variable)
   n27 = &Mixer_51_Two(
    1.InC=@n22.Out C
    1.InL=@n22.Out L
    1.InLF=@n22.Out LF
    1.InR=@n22.Out R
    1.InSL=@n22.Out SL
    1.InSR=@n22.Out SR
    2.InR=UE.Multiply:Audio(PrimaryOperand=@n25.Out R, AdditionalOperands=@n26.Value).out
    2.InL=UE.Multiply:Audio(PrimaryOperand=@n25.Out L, AdditionalOperands=@n26.Value).out
   ) {also {defaults {"2.Volume(Lin)"=1.5}}}
   n28 = &Mixer_51_Two(
    1.InL=@n19.Out L
    1.InR=@n19.Out R
    2.InC=@n27.OutC
    2.InL=@n27.OutL
    2.InLF=@n27.OutLF
    2.InR=@n27.OutR
    2.InSL=@n27.OutSL
    2.InSR=@n27.OutSR
   )
   n29 = &MSP_Audio_EQCrossFade_51(
    Input=&InterpSpeed(FloatInput=@TriggerRouteTriggerRouteFloat2_4.Value) {also {defaults {FallingInterpSpeed=3, RisingInterpSpeed=0.1}}}.FloatOutput
    In C=@n28.OutC
    In L=@n28.OutL
    In LF=@n28.OutLF
    In R=@n28.OutR
    In SL=@n28.OutSL
    In SR=@n28.OutSR
   ) {also {defaults {Max.DbAdjustment=-10, Max.LPF=2000}}}
   n30 = &MSP_Audio_EQCrossFade_51(
    In C=@n29.Out C
    In L=@n29.Out L
    In LF=@n29.Out LF
    In R=@n29.Out R
    In SL=@n29.Out SL
    In SR=@n29.Out SR
    Input=VariableAccessor.Float(Variable=@n17.Variable).Value
   ) {also {defaults {Interp Time=1, Max.DbAdjustment=2, Max.Value=1800, Min.Value=900}}}
   ADSREnvelopeADSREnvelope_1 = ADSR Envelope.ADSR Envelope:Audio(Trigger Attack=^UE.Source.OnPlay) {also {defaults {Attack Curve=2, Attack Time=0.7, Decay Time=0, Sustain Level=1}}}
   n31 = VariableAccessor.Audio(Variable=VariableMutator.Audio(Value=@ADSREnvelopeADSREnvelope_1.Out Envelope, Variable=InitVariable.Audio().Variable).Variable)
   Multiply_0 = UE.Multiply:Audio(PrimaryOperand=@n30.Out LF, AdditionalOperands=@n31.Value)
   Multiply_1 = UE.Multiply:Audio(PrimaryOperand=@n30.Out L, AdditionalOperands=@n31.Value)
   Multiply_2 = UE.Multiply:Audio(PrimaryOperand=@n30.Out R, AdditionalOperands=@n31.Value)
   Multiply_3 = UE.Multiply:Audio(PrimaryOperand=@n30.Out SR, AdditionalOperands=@n31.Value)
   Multiply_5 = UE.Multiply:Audio(PrimaryOperand=@n30.Out C, AdditionalOperands=@n31.Value)
   Multiply_6 = UE.Multiply:Audio(PrimaryOperand=@n30.Out SL, AdditionalOperands=@n31.Value)
  ]
  links [
   {dst="@UE.OutputFormat.5dot1.Audio:3", src=Multiply_0.out}
   {dst="@UE.OutputFormat.5dot1.Audio:0", src=Multiply_1.out}
   {dst="@UE.OutputFormat.5dot1.Audio:1", src=Multiply_2.out}
   {dst="@UE.OutputFormat.5dot1.Audio:5", src=Multiply_3.out}
   {dst="@UE.OutputFormat.5dot1.Audio:2", src=Multiply_5.out}
   {dst="@UE.OutputFormat.5dot1.Audio:4", src=Multiply_6.out}
  ]
 }]
properties {
 assetImportData=&self:AssetImportData
 attenuationSettings=/BarnTowelGameplay/Assets/Sounds/Settings/BarnTowel_InFlightAir_Att
 bLooping=true
 duration=10000
 maxDistance=3650
 numChannels=6
 outputFormat=FiveDotOne
 sampleRate=48000
 soundClassObject=&Fort_Gliders_InAir
 volume=0.2
}
