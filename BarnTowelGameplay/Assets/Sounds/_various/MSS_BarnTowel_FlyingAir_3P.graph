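# Looping mono third-person MetaSound for BarnTowel glider in-flight air, using glider sound class and attenuation.
# On play, a thrust loop starts and a delayed attack envelope fades in the whole mix.
# When isHovering changes to false, a random forward whoosh plays, with only its left channel mixed alongside the loop.
# A Doppler patch receives speed and attenuation distance, but its output feeds nothing in the graph.
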MetaSoundSource /BarnTowelGameplay/Assets/Sounds/_various/MSS_BarnTowel_FlyingAir_3P
inputs [
 3PAirLoop {type=WaveAsset:Array}
 isHovering {type=Bool}
 Speed {type=Float}
 StartFlyingOneshots {type=WaveAsset:Array}
 UE.Attenuation.Distance {type=Float}
 UE.Source.OnPlay {type=Trigger}
 UE.Source.Orientation.Azimuth {type=Float}
 UE.Source.Orientation.Elevation {type=Float}
]
interfaces [UE.Attenuation:1.0, UE.OutputFormat.Mono:1.0, UE.Source.Orientation:1.0, UE.Source:1.1]
outputs ["UE.OutputFormat.Mono.Audio:0" {type=Audio}]
pages [Default {
  input_defaults {
   Speed=0
   UE.Attenuation.Distance=0
   UE.Source.OnPlay=null
   UE.Source.Orientation.Azimuth=0
   UE.Source.Orientation.Elevation=0
   isHovering=null
   3PAirLoop [&/BarnTowell_Thrust_Loop_01]
   StartFlyingOneshots [
    &/BarnTowel_Forward_Whoosh_03
    &/BarnTowel_Forward_Whoosh_02
    &/BarnTowel_Forward_Whoosh_01
   ]
  }
  nodes [
   InitVariable.Float()
   /FortMetasound/Patches/MSP_Doppler(Speed=^Speed, Distance=^UE.Attenuation.Distance) {also {defaults {StrengthMultiplier=5}}}
   TriggerDelay_0 = UE.Trigger Delay(In=^UE.Source.OnPlay)
   ADSREnvelopeADSREnvelope_0 = ADSR Envelope.ADSR Envelope:Audio(Trigger Attack=@TriggerDelay_0.out) {also {defaults {Attack Time=0.3, Decay Time=0, Release Time=0, Sustain Level=1}}}
   ArrayRandomGet_0 = Array.Random Get:WaveAsset:Array(In Array=^3PAirLoop, Next=^UE.Source.OnPlay)
   WavePlayer_0 = UE.Wave Player:Mono(Play=@ArrayRandomGet_0.On Next, Wave Asset=@ArrayRandomGet_0.Value) {also {defaults {Loop="<MetaSoundBuilderResult.SUCCEEDED: 0>"}}}
   TriggerOnValueChange_0 = UE.Trigger On Value Change:Bool(Value=^isHovering)
   ArrayRandomGet_1 = Array.Random Get:WaveAsset:Array(In Array=^StartFlyingOneshots, Next=&MSP_Branch(Value=^isHovering, In=@TriggerOnValueChange_0.Trigger).False) {also {defaults {Weights []}}}
   WavePlayer_1 = UE.Wave Player:Stereo(Play=@ArrayRandomGet_1.On Next, Wave Asset=@ArrayRandomGet_1.Value)
   AudioMixerAudioMixerMono3_0 = "AudioMixer.Audio Mixer (Mono, 3)"(In 0=@WavePlayer_0.Out Mono, In 1=@WavePlayer_1.Out Left) {also {defaults {Gain 1=0.5, Gain 2=0.5}}}
   Reroute_0 = UE.Reroute(Value=UE.Multiply:Audio(PrimaryOperand=@ADSREnvelopeADSREnvelope_0.Out Envelope, AdditionalOperands=@AudioMixerAudioMixerMono3_0.Out).out)
  ]
  links [{dst="@UE.OutputFormat.Mono.Audio:0", src=Reroute_0.Value}]
 }]
properties {
 assetImportData=&self:AssetImportData
 attenuationSettings=/BarnTowelGameplay/Assets/Sounds/Settings/BarnTowel_InFlightAir_3P_Att
 bLooping=true
 duration=10000
 sampleRate=48000
 soundClassObject=&Fort_Gliders_InAir
 volume=1.5
}
